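; Battle Pass S16 quest ability that credits the owning player for chasing the chicken (Nug) for a total of seven seconds.
; It counts each addition of the NugChase active tag on the player pawn, which the chicken applies every second to nearby pawns through a short-lived effect.
; At seven counts it sends a NugChase stat event to the quest manager once, then ends.
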(event Ability|EventActivateAbility
  (bind _time__seconds (Variables|Default|Gettime%20%28seconds%29))
  (|CallParentFunction)
  (Ability|Tasks|WaitGameplayTagAddToActor (Variables|GAVars|Data|GetOwningPlayerPawn) "(TagName=\"Athena.Quests.NugChase.Active\")"
    (:Added
      (Math|Integer|IncrementInt _time__seconds
        (:%20%20
          (Development|PrintString (Utilities|String|Append "NUG +1 SECOND (total: " (Utilities|String|ToString(Integer) (Variables|Default|Gettime%20%28seconds%29)) ")") false)
          (if (>= _time__seconds 7)
            (Utilities|FlowControl|DoOnce false
              (:Completed
                (Quest|SendComplexCustomStatEvent (Variables|GAVars|Data|GetOwningPlayerQuestManager) 0 "(GameplayTags=((TagName=\"Athena.Quests.NugChase\")))" "(GameplayTags=)" 1)
                (Development|PrintString "NUG - CHASED THE CHICKEN FOR 7 SECONDS!!")
                (Ability|EndAbility)))))) ; +1 second
    )) ; This tag gets sent once a second to all nearby pawns\nfrom the chicken, but the GE only lasts 0.25 seconds.\n..that seems weird, but helps with making sure the tag\ngets cleaned up on nearby pawns.
)

(event Ability|EventFailedToActivatePassiveAbility
  (params (ActorInfo :type Gameplay%20Ability%20Actor%20Info%20Structure)))
